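# Translucent, two-sided unlit material for the Horned Judgment glider's wispy flame trail, usable on Niagara ribbons, sprites and skeletal meshes.
# Panned, distorted flame noise blends with a mask texture and UV-based width, tip and fade shaping.
# The result drives a curve-atlas three-color emissive ramp with time-of-day scaling, plus opacity with view-angle and depth fade; forward/mobile subtracts a mask instead of depth fading.
# Vertices wobble via panned noise and shift along camera direction.
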Material /BRCosmetics/Gadgets/Assets/VinderTech_GliderChute/Glider_HornedJudgment_Midgard/FX/M_Glider_HornedJudgement_Trail
graph {
 properties {
  bUsedWithMeshDeformer=true
  bUsedWithNiagaraRibbons=true
  bUsedWithNiagaraSprites=true
  bUsedWithSkeletalMesh=true
  blendMode=BLEND_Translucent
  pixelDepthOffsetMode=PDOM_Legacy
  shadingModel=MSM_Unlit
  twoSided=true
  usedShadingModels=Unlit
 }
 nodes [
  n1 = TextureCoordinate()
  n2 = ComponentMask(None=VectorParameter() {group=WPO, parameterName=WPO_UVScale, defaultValue=<rgba 1 1 0 0>}.RGB) {g=true, r=true}
  n3 = Constant()
  n4 = Panner(
   Coordinate=Add(A=Multiply(A=@n1, B=@n2), B=AppendVector(A=@n3, B=@n3)) {constB=0}
   Speed=ComponentMask(None=VectorParameter() {group=WPO, parameterName=WPO_Speed, defaultValue=<rgba 0 0.5 0 0>}.RGB) {g=true, r=true}
  ) {speedY=0.5}
  n5 = TextureCoordinate()
  n6 = Saturate(None=Power(Base=ComponentMask(None=@n5) {g=true}) {constExponent=1})
  n7 = Multiply(
   A=LinearInterpolate(Alpha=TextureSample(UVs=@n4) {samplerType=SAMPLERTYPE_Normal, texture=&T_Weap_Distort_N_02}.R) {constB=1.5}
   B=@n6
  )
  n8 = VectorParameter() {group=WPO, parameterName=WPO_direction, defaultValue=<rgba 1 0 0 1>}
  n9 = MaterialFunctionCall() {materialFunction=&CameraDirectionVector}
  n10 = Multiply(A=ScalarParameter() {parameterName=Camera Offset}) {constB=-1}
  Add(
   A=Transform(None=Multiply(A=@n7, B=@n8.RGB)) {transformSourceType=TRANSFORMSOURCE_Local}
   B=Multiply(A=@n6, B=Multiply(A=@n9.Result, B=@n10))
  )
  n11 = ScalarParameter() {defaultValue=2, group=Color, parameterName=Light Multiplier, sliderMax=5}
  n12 = VectorParameter() {parameterName=TOD Scalar, defaultValue=<rgba 1 1 1 1>}
  n13 = Multiply(
   A=VectorParameter() {parameterName=AdvancedEmissiveColor, defaultValue=<rgba 0 0.1833 0.4896 1>}.RGB
   B=MaterialFunctionCall(Value_Morning=@n12.R, Value_Day=@n12.R, Value_Evening=@n12.R, Value_Night=@n12.R) {materialFunction=&Scalar_Time_of_Day}.Result
  )
  n14 = ScalarParameter() {defaultValue=1, group=Color, parameterName=Middle Multiplier, sliderMax=5}
  n15 = Desaturation(
   None=MaterialFunctionCall(
    Hue Shift Percentage=ScalarParameter() {defaultValue=0.02, group=Color, parameterName=Dark Hue Shift, sliderMax=1, sliderMin=-1}
    Texture=@n13
   ) {materialFunction=&HueShift}.Result
   Fraction=ScalarParameter() {group=Color, parameterName=Desaturation, sliderMax=2, sliderMin=-2}
  )
  n16 = ScalarParameter() {defaultValue=0.5, group=Color, parameterName=Dark Multiplier, sliderMax=5}
  n17 = TextureCoordinate()
  n18 = StaticSwitchParameter(True=OneMinus(None=ComponentMask(None=@n17) {g=true}), False=OneMinus(None=ComponentMask(None=@n17) {r=true})) {parameterName=Swap UV}
  n19 = Power(Base=@n18, Exp=ScalarParameter() {parameterName=End Fade}) {constExponent=0}
  n20 = TextureSample(UVs=Add(A=TextureCoordinate(), B=Constant2Vector())) {
   samplerType=SAMPLERTYPE_LinearGrayscale
   texture=&/T_Glider_HornedJudgement_Mask
  }
  n21 = ScalarParameter() {defaultValue=1, parameterName=Mask Strength, sliderMax=5}
  n22 = TextureCoordinate()
  n23 = AppendVector(
   A=ScalarParameter() {defaultValue=1, group=Wispy Flame, parameterName=Distortion Scale X, sliderMax=5}
   B=ScalarParameter() {defaultValue=1, group=Wispy Flame, parameterName=Distortion Scale Y, sliderMax=5}
  )
  n24 = ScalarParameter() {defaultValue=-1, group=Wispy Flame, parameterName=Distortion Speed X, sliderMax=5, sliderMin=-5}
  n25 = ScalarParameter() {group=Wispy Flame, parameterName=Distortion Speed Y, sliderMax=5, sliderMin=-5}
  n26 = TextureSampleParameter2D(UVs=Panner(Coordinate=Multiply(A=@n22, B=@n23), Speed=AppendVector(A=@n24, B=@n25)) {speedX=1}) {
   group=Wispy Flame
   parameterName=Flame Distortion Texture
   samplerType=SAMPLERTYPE_LinearGrayscale
   texture=/Game/Effects/Fort_Effects/Emotes/UltraLight/Texture/T_FluidFlameTile_Greyscale
  }
  n27 = ScalarParameter() {defaultValue=0.1, parameterName=Distortion Strength, sliderMax=1}
  n28 = Add(A=TextureCoordinate(), B=Multiply(A=Multiply(A=@n26.R, B=@n27) {constB=0.1}))
  n29 = AppendVector(
   A=ScalarParameter() {defaultValue=1, group=Wispy Flame, parameterName=Scale X, sliderMax=5}
   B=ScalarParameter() {defaultValue=1, group=Wispy Flame, parameterName=Scale Y, sliderMax=5}
  )
  n30 = ScalarParameter() {defaultValue=-1, group=Wispy Flame, parameterName=Speed X, sliderMax=5, sliderMin=-5}
  n31 = ScalarParameter() {group=Wispy Flame, parameterName=Speed Y, sliderMax=5, sliderMin=-5}
  n32 = TextureSampleParameter2D(UVs=Panner(Coordinate=Multiply(A=@n28, B=@n29), Speed=AppendVector(A=@n30, B=@n31)) {speedX=1}) {
   group=Wispy Flame
   parameterName=Flame Texture
   samplerType=SAMPLERTYPE_Grayscale
   texture=/Game/Effects/Fort_Effects/Textures/Fire/T_WispyFireNoise_Soft_M
  }
  n33 = TextureCoordinate()
  n34 = ComponentMask(None=@n33) {r=true}
  n35 = ComponentMask(None=@n33) {g=true}
  n36 = Power(
   Base=SphereMask(
    A=StaticSwitchParameter(True=@n34, False=@n35) {parameterName=Swap UV}
    B=Constant() {r=0.5}
    Radius=ScalarParameter() {defaultValue=0.5, parameterName=Width, sliderMax=1}
   ) {attenuationRadius=0.5, hardnessPercent=0}
   Exp=ScalarParameter() {defaultValue=0.8, parameterName=Side Fade, sliderMax=10}
  )
  n37 = Power(Base=@n18, Exp=ScalarParameter() {parameterName=Start Fade Intensity, sliderMax=5}) {constExponent=30}
  n38 = Constant()
  n39 = SmoothStep(
   Min=ScalarParameter() {parameterName=Pointiness, sliderMax=1}
   Value=Subtract(A=Multiply(A=@n18, B=@n36), B=StaticSwitchParameter(True=@n37, False=@n38) {parameterName=Start Fade?})
  )
  n40 = ScalarParameter() {defaultValue=3, parameterName=Shape Strength, sliderMax=5}
  n41 = Add(A=@n32.R, B=ScalarParameter() {parameterName=Gradient base, sliderMax=1}) {constB=0.2}
  n42 = LinearInterpolate(
   A=Subtract(A=Multiply(A=@n20.R, B=@n21), B=@n32.R)
   B=Multiply(A=Multiply(A=@n39, B=@n40) {constB=3}, B=Saturate(None=@n41))
   Alpha=Saturate(None=Multiply(A=@n20.R, B=@n39) {constA=1})
  )
  n43 = CurveAtlasRowParameter(CurveTime=Multiply(A=@n19, B=@n42)) {
   atlas=&/CA_Glider_HornedJudgement
   curve=&/C_Glider_HornedJudgement
   parameterName=Curve_Color
  }
  MaterialFunctionCall(Emissive=MaterialFunctionCall(A=Multiply(A=@n13, B=@n11) {constB=2}, B=Multiply(A=@n13, B=@n14) {constB=1.5}, C=Multiply(A=@n15, B=@n16), Alpha=@n43.R) {materialFunction=&Lerp_3Color}.Result) {materialFunction=&MF_FX_EmissiveAdjustments}
  n44 = Saturate(None=Multiply(A=@n19, B=@n42))
  n45 = MaterialFunctionCall("FadeSensitivity (optional)"=ScalarParameter() {defaultValue=4, parameterName=ViewAngleFadeSensitivity, sliderMax=10}) {materialFunction=&FadeBasedOnViewAngle}
  n46 = SmoothStep(
   Max=OneMinus(None=ScalarParameter() {defaultValue=0.9, parameterName=Edge Hardness, sliderMax=1})
   Value=Multiply(A=Multiply(A=@n44, B=@n45.Result))
  ) {constMax=0.1}
  n47 = DepthFade(Opacity=@n46, FadeDistance=ScalarParameter() {defaultValue=35, parameterName=Depth Fade Distance, sliderMax=100}) {fadeDistanceDefault=25}
  n48 = VectorParameter() {defaultValue=<rgba 0 0.2 0 1>}
  n49 = TextureSample(UVs=Add(A=TextureCoordinate(), B=AppendVector(A=@n48.R, B=@n48.R))) {
   samplerType=SAMPLERTYPE_LinearGrayscale
   texture=&/T_Glider_HornedJudgement_MaskForMobile
  }
  n50 = Subtract(A=@n46, B=Power(Base=@n49.R, Exp=ScalarParameter() {defaultValue=1, parameterName=Exp}) {constExponent=0.5})
  ShadingPathSwitch(Default=@n47, Deferred=@n47, Forward=@n50, Mobile=@n50)
 ]
}
