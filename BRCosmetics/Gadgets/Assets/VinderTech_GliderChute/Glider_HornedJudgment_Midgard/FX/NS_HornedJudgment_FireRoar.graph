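# Niagara fire-roar effect for the Horned Judgment Midgard glider, layering meshes and sprites into a short forward blast.
# Particle_Emitter bursts a fading mesh, while GoldenNoise adds drifting meshes with animated material parameters.
# SparksGPU throws speed-stretched GPU sparks and Spit launches falling sprites, with emitter-age curves tapering spawn rates by about 2.5 seconds.
# All emitters take User.AdvancedEmissiveColor, alongside system-level variant hue extraction and color reconstruction modules.
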NiagaraSystem /BRCosmetics/Gadgets/Assets/VinderTech_GliderChute/Glider_HornedJudgment_Midgard/FX/NS_HornedJudgment_FireRoar
dependencies {
 usedDataInterfaces [
  &NiagaraDataInterfaceCurve
  &NiagaraDataInterfaceVector2DCurve
  &NiagaraDataInterfaceArrayFloat
  &NiagaraDataInterfaceVectorCurve
 ]
 usedDynamicInputs []
 usedModules [
  &self:SystemSpawnScript.SystemScriptSource.SystemScriptGraph.NiagaraNodeAssignment_1.SetVariables_#1
  &self:ExtractVariantHue
  &self:SystemSpawnScript.SystemScriptSource.SystemScriptGraph.NiagaraNodeAssignment_0.SetVariables_#2
  &self:ReconstructColors
  &SystemState
 ]
}
emitters {
 GoldenNoise {
  emitterUpdateScript [
   &EmitterState()
   &SpawnBurst_Instantaneous(Use Spawn Probability=false, UseLoopCountLimit=false)
   &SpawnRate(
    SpawnRate=&Multiply_Float(
     A=10
     B=&FloatFromCurve(
      FloatCurve=DataInterface() {propertyValues {Curve {keys [{interpMode=RCIM_Cubic, tangentMode=RCTM_User, time=1.5, value=1}, {interpMode=RCIM_Cubic, tangentMode=RCTM_User, time=2.5, value=0}]}}}
      CurveIndex=$Emitter.Age
     )
    )
    Use Spawn Probability=false
   )
  ]
  particleSpawnScript [
   &InitializeParticle(
    Color=$User.AdvancedEmissiveColor
    Lifetime=&RandomRangeFloat(Minimum=0.3, Maximum=0.6)
    Mesh Scale=<1.57 1.57 1.57>
    Mesh Scale Mode=::NewEnumerator4
    Position=&ConvertVectorToPosition(
     Input Position=&RandomRangeVector(Minimum=<50 0 0>, Maximum=<200 0 0>)
     Conversion Operation:ENiagara_LWCConvertVecToPos=::NewEnumerator0
    )
    Position Mode=::NewEnumerator1
    UsePositionOffset=false
    Write Material Random=true
   )
   &Color(Color=&MakeLinearColorFromVectorAndFloat(
     "Vector (RGB)"=<13.5 10.1565 1.2656>
     "Float (Alpha)"=&ClampFloat(Float=&FloatFromCurve(FloatCurve=DataInterface() {propertyValues {Curve {keys [<at 0 2 RCIM_Cubic>, {interpMode=RCIM_Cubic, leaveTangent=-1.6594, time=2, value=0}]}}}))
    )) {enabled=false}
   /Niagara/Modules/CascadeConversion/CascadeConversion_InitialMeshRotation(Rotation:Vector3f=&RandomRangeVector(Minimum=<0 0 0>, Maximum=<1 1 1>))
   &AddVelocity(Velocity=&RandomRangeVector(Minimum=<100 0 0>, Maximum=<150 0 0>))
  ]
  particleUpdateScript [
   &ParticleState()
   &ScaleMeshSize(Scale Factor=&VectorFromFloat(Value=&FloatFromCurve(FloatCurve=DataInterface() {propertyValues {Curve {keys [{arriveTangent=1.0383, interpMode=RCIM_Cubic, time=-0.0063, value=0.25}, <at 1 1 RCIM_Cubic>]}}})))
   &DynamicMaterialParameters(
    Index 0 Param 1=&FloatFromCurve(FloatCurve=DataInterface() {propertyValues {Curve {keys [<at 0 0.5>,<at 1 0.15>]}}})
    Index 0 Param 4=&FloatFromCurve(FloatCurve=DataInterface() {propertyValues {Curve {keys [<at 0 1>,<at 1 4>]}}})
    Module.Index0Param0Write=true
    Module.Index0Param1Write=false
    Module.Index0Param2Write=false
    Module.Index0Param3Write=true
   )
   &ScaleColor(
    Scale Alpha=&FloatFromCurve(FloatCurve=DataInterface() {propertyValues {Curve {keys [<at 0 0 RCIM_Cubic>,<at 0.0987 1 RCIM_Cubic>,<at 1 0>]}}})
    Scale RGB=<10 10 10>
    ScaleA=true
    ScaleRGB=true
   )
   &SolveForcesAndVelocity(Clamp Velocity=false, Limit Acceleration=false)
  ]
  rendererClasses [&NiagaraMeshRendererProperties]
  renderers [{rendererClass=&NiagaraMeshRendererProperties, rendererIndex=0}]
 }
 Particle_Emitter {
  emitterUpdateScript [
   &EmitterState()
   &SpawnBurst_Instantaneous(Use Spawn Probability=false, UseLoopCountLimit=false)
  ]
  particleSpawnScript [&InitializeParticle(
    Color=$User.AdvancedEmissiveColor
    Lifetime=2.5
    Mesh Scale=<2 1 1>
    Mesh Scale Mode=::NewEnumerator4
    UsePositionOffset=false
    Write Material Random=true
   )]
  particleUpdateScript [
   &ParticleState()
   &Timeline(Additional Output:ENiagara_TimelineOutput=::NewEnumerator1, Additional Output Mode:ENiagara_TimelineOutputMode=::NewEnumerator1)
   &ScaleMeshSize(Scale Factor=&VectorFromCurve(VectorCurve=DataInterface() {propertyValues {
       XCurve {keys [<at 0 0.5 RCIM_Cubic>, {arriveTangent=0.5, interpMode=RCIM_Cubic, leaveTangent=0.5, time=0.15, value=1}, <at 1 1>]}
       YCurve {keys [<at 0 1 RCIM_Cubic>,<at 0.15 1 RCIM_Cubic>,<at 1 1>]}
       ZCurve {keys [<at 0 1 RCIM_Cubic>,<at 0.15 1 RCIM_Cubic>,<at 1 1>]}
      }}))
   &Color(Color=&MakeLinearColorFromVectorAndFloat(
     "Vector (RGB)"=&VectorFromCurve(VectorCurve=DataInterface() {propertyValues {XCurve {keys [<at 0 1>,<at 1 1>]}, YCurve {keys [<at 0 1>,<at 1 1>]}, ZCurve {keys [<at 0 1>,<at 1 1>]}}})
     "Float (Alpha)"=&ClampFloat(Float=&FloatFromCurve(FloatCurve=DataInterface() {propertyValues {Curve {keys [<at 0 0 RCIM_Cubic>,<at 0.15 1 RCIM_Cubic>,<at 1 0>]}}}))
    )) {enabled=false}
   &ScaleColor(
    Scale Alpha=&FloatFromCurve(FloatCurve=DataInterface() {propertyValues {Curve {keys [<at 0 0 RCIM_Cubic>, {arriveTangent=2, interpMode=RCIM_Cubic, leaveTangent=2, time=0.15, value=1}, <at 0.5 1>, <at 1 0>]}}})
    ScaleA=true
    ScaleRGB=false
   )
  ]
  rendererClasses [&NiagaraMeshRendererProperties]
  renderers [{rendererClass=&NiagaraMeshRendererProperties, rendererIndex=0}]
 }
 SparksGPU {
  simTarget=GPUComputeSim
  emitterUpdateScript [
   &EmitterState()
   &SpawnRate(
    SpawnRate=&Multiply_Float(
     A=100
     B=&FloatFromCurve(
      FloatCurve=DataInterface() {propertyValues {Curve {keys [{interpMode=RCIM_Cubic, tangentMode=RCTM_User, time=2, value=1}, {interpMode=RCIM_Cubic, tangentMode=RCTM_User, time=2.5, value=0}]}}}
      CurveIndex=$Emitter.Age
     )
    )
    Use Spawn Probability=false
   )
  ]
  particleSpawnScript [
   &InitializeParticle(
    Color=$User.AdvancedEmissiveColor
    Lifetime=&RandomRangeFloat(Minimum=0.2, Maximum=0.6)
    Position=&ConvertVectorToPosition(
     Input Position=&RandomRangeVector(Minimum=<25 0 0>, Maximum=<100 0 0>)
     Conversion Operation:ENiagara_LWCConvertVecToPos=::NewEnumerator0
    )
    Position Mode=::NewEnumerator1
    Sprite Size=&RandomRangeVector2D(Minimum=<1 1.5>, Maximum=<1.5 3>)
    Sprite Size Mode=::NewEnumerator4
    UsePositionOffset=false
    Write Material Random=true
   )
   &AddVelocity(Velocity=&RandomRangeVector(Minimum=<0 0 0>, Maximum=<0 0 300>)) {enabled=false}
   &ShapeLocation(MaskSpawnGroup=false, Override Seed=false, Sphere Radius=5, UseOwnerScale=true)
   &AddVelocity(
    Inner Cone Angle=25
    Module.UseConeFalloff=true
    Module.UseVelDistribution=true
    Override Seed=false
    Velocity Mode=::NewEnumerator2
    Velocity Speed=&RandomRangeFloat(Minimum=500, Maximum=1000)
   ) {module=AddVelocity001}
  ]
  particleUpdateScript [
   &ParticleState()
   &ScaleSpriteSize(
    Scale Factor=&Vector2DFromCurve(
     Vector2Curve=DataInterface() {propertyValues {
       XCurve {keys [<at 0 1>, {arriveTangent=-0.7018, leaveTangent=-0.7018, time=0.1, value=1}, <at 1.0008 0.2982>]}
       YCurve {keys [{arriveTangent=1, time=0, value=0}, {arriveTangent=-1, leaveTangent=-1, time=0.1, value=1}, <at 1.0008 0>]}
      }}
     ScaleCurve=<1 1>
    )
    Scale Sprite Size Mode=::NewEnumerator2
   )
   &Color(Color=&MakeLinearColorFromVectorAndFloat(
     "Vector (RGB)"=<5 5 5>
     "Float (Alpha)"=&ClampFloat(Float=&FloatFromCurve(FloatCurve=DataInterface() {propertyValues {Curve {keys [{arriveTangent=-1, time=0, value=1}, <at 1 0>]}}}))
    )) {enabled=false}
   &Drag(Drag=1.5, Use Linear Drag=true, Use Rotational Drag=true) {enabled=false}
   &AccelerationForce(Acceleration=<0 0 -200>) {enabled=false}
   &SolveForcesAndVelocity(Clamp Velocity=false, Limit Acceleration=false)
   &ScaleSpriteSizeBySpeed(Max Scale Factor=<10 100>, Source Velocity:Vector3f=&Multiply_Vector(A=<0 0.02 1>, B=$Particles.Velocity))
   &ScaleColor(
    Scale Alpha=&ClampFloat(Float=&FloatFromCurve(FloatCurve=DataInterface() {propertyValues {Curve {keys [{arriveTangent=-1, time=0, value=1}, <at 1 0>]}}}))
    Scale RGB=<5 5 5>
    ScaleA=true
    ScaleRGB=true
   )
  ]
  rendererClasses [&NiagaraSpriteRendererProperties]
  renderers [{rendererClass=&NiagaraSpriteRendererProperties}]
 }
 Spit {
  emitterUpdateScript [
   &EmitterState()
   &SpawnRate(
    SpawnRate=&Multiply_Float(
     A=12
     B=&FloatFromCurve(
      FloatCurve=DataInterface() {propertyValues {Curve {keys [{interpMode=RCIM_Cubic, tangentMode=RCTM_User, time=2, value=1}, {interpMode=RCIM_Cubic, tangentMode=RCTM_User, time=2.5, value=0}]}}}
      CurveIndex=$Emitter.Age
     )
    )
    Use Spawn Probability=false
   ) {module=SpawnRate001}
  ]
  particleSpawnScript [
   &InitializeParticle(
    Color=$User.AdvancedEmissiveColor
    Lifetime=&RandomRangeFloat(Minimum=0.35, Maximum=0.5)
    Sprite Size=&Multiply_Vector2D_ByFloat(
     Float=&FloatFromCurve(
      FloatCurve=DataInterface() {propertyValues {Curve {keys [{interpMode=RCIM_Cubic, tangentMode=RCTM_User, time=1.8, value=1}, {interpMode=RCIM_Cubic, tangentMode=RCTM_User, time=2.8, value=0}]}}}
      CurveIndex=$Emitter.Age
     )
     Vector2D=&RandomRangeVector2D(Minimum=<25 45>, Maximum=<75 75>)
    )
    Sprite Size Mode=::NewEnumerator4
    UsePositionOffset=false
    Write Material Random=true
   )
   &AddVelocity(Velocity=&RandomRangeVector(Minimum=<300 -100 -50>, Maximum=<700 100 100>), Velocity Speed Scale=2)
   &Color(Color=&MakeLinearColorFromVectorAndFloat(
     "Vector (RGB)"=&VectorFromFloat(Value=&RandomRangeFloat(Minimum=1, Maximum=2))
     "Float (Alpha)"=&ClampFloat(Float=1)
    )) {enabled=false}
  ]
  particleUpdateScript [
   &ParticleState()
   &ScaleSpriteSize(
    Scale Factor=&Vector2DFromCurve(
     Vector2Curve=DataInterface() {propertyValues {
       XCurve {keys [<at 0 0 RCIM_Cubic>,<at 0.7 1 RCIM_Cubic>,<at 1 0 RCIM_Cubic>]}
       YCurve {keys [<at 0 0 RCIM_Cubic>,<at 0.7 3.625 RCIM_Cubic>,<at 1 0 RCIM_Cubic>]}
      }}
     ScaleCurve=<1 1>
    )
    Scale Sprite Size Mode=::NewEnumerator2
   )
   &AccelerationForce(Acceleration=<0 0 -500>)
   &SolveForcesAndVelocity(Clamp Velocity=false, Limit Acceleration=false)
   &ScaleColor(
    Scale Alpha=&FloatFromCurve(FloatCurve=DataInterface() {propertyValues {Curve {keys [{interpMode=RCIM_Cubic, tangentMode=RCTM_User, time=0, value=1}, {interpMode=RCIM_Cubic, tangentMode=RCTM_User, time=1, value=0}]}}})
    Scale RGB=<2 2 2>
    ScaleA=true
    ScaleRGB=true
   )
   &DynamicMaterialParameters(
    Index 0 Param 3=1
    Index 0 Param 4=2
    Module.Index0Param0Write=true
    Module.Index0Param1Write=false
    Module.Index0Param2Write=false
    Module.Index0Param3Write=false
   )
  ]
  rendererClasses [&NiagaraSpriteRendererProperties]
  renderers [{rendererClass=&NiagaraSpriteRendererProperties}]
 }
}
summary {
 systemName=NS_HornedJudgment_FireRoar
 emitters [
  {emitterName=Spit, rendererClasses [&NiagaraSpriteRendererProperties]}
  {emitterName=Particle_Emitter, rendererClasses [&NiagaraMeshRendererProperties]}
  {emitterName=SparksGPU, simTarget=GPUComputeSim, rendererClasses [&NiagaraSpriteRendererProperties]}
  {emitterName=GoldenNoise, rendererClasses [&NiagaraMeshRendererProperties]}
 ]
 userVariables [User.AdvancedEmissiveColor {type=LinearColor, description="", defaultValue {struct=LinearColor, value=<rgba 0.0194 0.2727 0.526 1>}}]
}
